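# Sound cue for the GateHound glider's forward thrust, combining a thrust sound with a looping acceleration layer.
# A mixer blends one of three random thrust waves, with slight pitch and volume modulation, with one of three random looping acceleration waves.
# An enveloper ramps the loop's volume up from silence, and the mixer weights it above the thrust layer.
# Shared glider attenuation, sound class and concurrency assets govern playback.
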SoundCue /BRCosmetics/Sounds/Gliders/GateHound/GliderThrustFwd_GateHound_Cue
root=Mixer_0
nodes [
 {
  id=Enveloper_1
  children [Random_1]
  properties {pitchCurve {editorCurveData {keys [{time=0}]}}, volumeCurve {editorCurveData {keys [<at 0 0>,<at 1 1>]}}}
 }
 {id=Mixer_0, children [Modulator_0,Enveloper_1], properties {inputVolume [0.8,1.1]}}
 {id=Modulator_0, children [Random_0], properties {pitchMax=1.02, pitchMin=0.98, volumeMax=1.02, volumeMin=0.98}}
 {id=Random_0, children [WavePlayer_0,WavePlayer_1,WavePlayer_2], properties {weights [1,1,1]}}
 {id=Random_1, children [WavePlayer_3,WavePlayer_4,WavePlayer_5], properties {weights [1,1,1]}}
 {id=WavePlayer_0, properties {soundWaveAssetPtr=&/Glider_GateHound_ThrustFwd_01}}
 {id=WavePlayer_1, properties {soundWaveAssetPtr=&/Glider_GateHound_ThrustFwd_02}}
 {id=WavePlayer_2, properties {soundWaveAssetPtr=&/Glider_GateHound_ThrustFwd_03}}
 {id=WavePlayer_3, properties {bLooping=true, soundWaveAssetPtr=&/Glider_GateHound_Fwd_LP_Accel}}
 {id=WavePlayer_4, properties {bLooping=true, soundWaveAssetPtr=&/Glider_GateHound_Fwd_LP_Accel_03}}
 {id=WavePlayer_5, properties {bLooping=true, soundWaveAssetPtr=&/Glider_GateHound_Fwd_LP_Accel_02}}
]
properties {
 attenuationSettings=&Fort_Player_Glider_Thrust_Athena_Att
 audioPropertiesSheet=&APS_Cosmetic_Glider_Loop
 maxDistance=12500
 soundClassObject=&Fort_Gliders_Land
 volumeMultiplier=1
 concurrencySet [&Athena_Glider_Thrust_Conc]
}
