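# Looping thrust sound cue for the GateHound glider cosmetic, played with glider thrust attenuation and the glider loop audio property sheet.
# A mixer blends three looping layers: idle, forward flame and forward main.
# The Forwards parameter continuously modulates each layer, lowering idle volume as it rises, driving both forward layers' volume and slightly raising pitch on all three.
# An enveloper adds a volume ramp-in to the main forward layer only.
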SoundCue /BRCosmetics/Sounds/Gliders/GateHound/GliderThrustLoop_GateHound_Cue
root=Mixer_0
nodes [
 {
  id=Enveloper_1
  children [ModulatorContinuous_4]
  properties {pitchCurve {editorCurveData {keys [{time=0}]}}, volumeCurve {editorCurveData {keys [<at 0 0>,<at 1 1>]}}}
 }
 {id=Mixer_0, children [ModulatorContinuous_0, ModulatorContinuous_2, Enveloper_1], properties {inputVolume [0.5,0.6,0.6]}}
 {
  id=ModulatorContinuous_0
  children [WavePlayer_0]
  properties {
   pitchModulationParams {maxOutput=1.1, minOutput=1, parameterName=Forwards}
   volumeModulationParams {maxOutput=0.4, minOutput=0.9, parameterName=Forwards}
  }
 }
 {
  id=ModulatorContinuous_2
  children [WavePlayer_2]
  properties {pitchModulationParams {maxOutput=1.1, minOutput=1, parameterName=Forwards}, volumeModulationParams {parameterName=Forwards}}
 }
 {
  id=ModulatorContinuous_4
  children [WavePlayer_4]
  properties {pitchModulationParams {maxOutput=1.1, minOutput=1, parameterName=Forwards}, volumeModulationParams {parameterName=Forwards}}
 }
 {id=WavePlayer_0, properties {bLooping=true, soundWaveAssetPtr=&/Glider_GateHound_Idle_LP}}
 {id=WavePlayer_2, properties {bLooping=true, soundWaveAssetPtr=&/Glider_GateHound_Fwd_LP_Flame}}
 {id=WavePlayer_4, properties {bLooping=true, soundWaveAssetPtr=&/Glider_GateHound_Fwd_LP_Main}}
]
properties {
 attenuationSettings=&Fort_Player_Glider_Thrust_Athena_Att
 audioPropertiesSheet=&APS_Cosmetic_Glider_Loop
 maxDistance=12500
 soundClassObject=&Fort_Gliders_Land
 volumeMultiplier=0.9
 concurrencySet [null]
}
